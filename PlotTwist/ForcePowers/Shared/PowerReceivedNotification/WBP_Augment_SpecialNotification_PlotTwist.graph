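# Star Wars Force power notification widget that announces completed Jedi or Sith training when an augment is acquired.
# After a delay (SprintJump is skipped), UpdateVisuals reads AugmentTypeData to set title, colors, glow material and async-loaded icon, then plays a desktop or touch intro.
# A close timer plays the outro, shortening both when more events are queued.
# Playlist tags offset its position, and SetMobileStyle adjusts icon scale and padding.
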WidgetBlueprint /PlotTwist/ForcePowers/Shared/PowerReceivedNotification/WBP_Augment_SpecialNotification_PlotTwist
animations [
 Intro {
  binding_count=7
  length_seconds=3.0982
  object=&self:Intro
  sequence {bindings [
    SizeBoxTriangles {
     id=#1
     possessed_class=SizeBox
     tracks [
      {
       type=MovieSceneFloatTrack
       display_name=Render Opacity
       sections [{type=MovieSceneFloatSection, channels [None {type=MovieSceneScriptingFloatChannel, keys [[25.4643,1],[40.4497,1],[61.9627,0]]}]}]
      }
      {
       type=MovieSceneFloatTrack
       display_name=Width Override
       sections [{type=MovieSceneFloatSection, channels [None {type=MovieSceneScriptingFloatChannel, keys [[25.4643,121.8743],[61.9627,425.1152]]}]}]
      }
      {
       type=MovieSceneFloatTrack
       display_name=Height Override
       sections [{type=MovieSceneFloatSection, channels [None {type=MovieSceneScriptingFloatChannel, keys [[25.4643,13.8272],[34.6563,49.1371]]}]}]
      }
     ]
    }
    Text_Title {
     id=#2
     possessed_class=CommonTextBlock
     tracks [
      {
       type=MovieScene2DTransformTrack
       sections [{
         type=MovieScene2DTransformSection
         channels [
          Scale.X {type=MovieSceneScriptingFloatChannel, keys [[18.795,10],[23.049,0.8094],[23.8003,1],[25.4643,1]]}
          Scale.Y {type=MovieSceneScriptingFloatChannel, keys [[18.795,10],[23.049,0.8094],[23.8003,1],[25.4643,1]]}
         ]
        }]
      }
      {
       type=MovieSceneFloatTrack
       display_name=Render Opacity
       sections [{type=MovieSceneFloatSection, channels [None {type=MovieSceneScriptingFloatChannel, keys [[18.795,0],[22.2427,1]]}]}]
      }
     ]
    }
    Overlay-AugmentBG {
     id=#3
     display_name=Overlay-AccoladeBG
     possessed_class=Image
     tracks [{
       type=MovieSceneFloatTrack
       display_name=Render Opacity
       sections [{type=MovieSceneFloatSection, channels [None {type=MovieSceneScriptingFloatChannel, keys [[25.4643,0],[32.1337,1]]}]}]
      }]
    }
    Text_EventDescription {
     id=#4
     possessed_class=CommonTextBlock
     tracks [
      {
       type=MovieSceneFloatTrack
       display_name=Render Opacity
       sections [{type=MovieSceneFloatSection, channels [None {type=MovieSceneScriptingFloatChannel, keys [[40.4497,0],[43.8973,1]]}]}]
      }
      {
       type=MovieScene2DTransformTrack
       sections [{
         type=MovieScene2DTransformSection
         channels [
          Scale.X {type=MovieSceneScriptingFloatChannel, keys [[40.4497,8],[44.7037,0.8094],[45.7963,0.9959],[47.119,1]]}
          Scale.Y {type=MovieSceneScriptingFloatChannel, keys [[40.4497,8],[44.7037,0.8094],[45.7963,0.9959],[47.119,1]]}
         ]
        }]
      }
     ]
    }
    Image_Glow {
     id=#5
     possessed_class=Image
     tracks [
      {
       type=MovieScene2DTransformTrack
       sections [{
         type=MovieScene2DTransformSection
         channels [
          Scale.X {type=MovieSceneScriptingFloatChannel, keys [[25.854,1],[29.7603,1],[47.119,1.2]]}
          Scale.Y {type=MovieSceneScriptingFloatChannel, keys [[25.854,1],[29.7603,1],[47.119,1.1]]}
         ]
        }]
      }
      {
       type=MovieSceneFloatTrack
       display_name=Render Opacity
       sections [{type=MovieSceneFloatSection, channels [None {type=MovieSceneScriptingFloatChannel, keys [[25.854,1],[61.9627,0]]}]}]
      }
     ]
    }
    Overlay-AugmentIcon {
     id=#6
     possessed_class=Overlay
     tracks [{
       type=MovieScene2DTransformTrack
       sections [{
         type=MovieScene2DTransformSection
         channels [
          Scale.X {type=MovieSceneScriptingFloatChannel, keys [[0,5],[3.8273,0.8643],[13.6013,1]]}
          Scale.Y {type=MovieSceneScriptingFloatChannel, keys [[0,5],[3.8273,0.8643],[13.6013,1]]}
         ]
        }]
      }]
    }
    "Text_Title (4)" {
     id=#7
     possessed_class=VerticalBox
     tracks [{
       type=MovieSceneFloatTrack
       display_name=Render Opacity
       sections [{type=MovieSceneFloatSection, channels [None {type=MovieSceneScriptingFloatChannel, keys [[0,0],[6,1]]}]}]
      }]
    }
   ]}
 }
 Intro_Mobile {
  binding_count=7
  length_seconds=3.0982
  object=&self:Intro_Mobile
  sequence {bindings [
    SizeBoxTriangles {
     id=#1
     possessed_class=SizeBox
     tracks [
      {
       type=MovieSceneFloatTrack
       display_name=Render Opacity
       sections [{type=MovieSceneFloatSection, channels [None {type=MovieSceneScriptingFloatChannel, keys [[25.4643,1],[40.4497,1],[61.9627,0]]}]}]
      }
      {
       type=MovieSceneFloatTrack
       display_name=Width Override
       sections [{type=MovieSceneFloatSection, channels [None {type=MovieSceneScriptingFloatChannel, keys [[25.4643,121.8743],[61.9627,425.1152]]}]}]
      }
      {
       type=MovieSceneFloatTrack
       display_name=Height Override
       sections [{type=MovieSceneFloatSection, channels [None {type=MovieSceneScriptingFloatChannel, keys [[25.4643,13.8272],[34.6563,49.1371]]}]}]
      }
     ]
    }
    Text_Title {
     id=#2
     possessed_class=CommonTextBlock
     tracks [
      {
       type=MovieScene2DTransformTrack
       sections [{
         type=MovieScene2DTransformSection
         channels [
          Scale.X {type=MovieSceneScriptingFloatChannel, keys [[18.795,10],[23.049,0.8094],[23.8003,1],[25.4643,1]]}
          Scale.Y {type=MovieSceneScriptingFloatChannel, keys [[18.795,10],[23.049,0.8094],[23.8003,1],[25.4643,1]]}
         ]
        }]
      }
      {
       type=MovieSceneFloatTrack
       display_name=Render Opacity
       sections [{type=MovieSceneFloatSection, channels [None {type=MovieSceneScriptingFloatChannel, keys [[18.795,0],[22.2427,1]]}]}]
      }
     ]
    }
    Overlay-AugmentBG {
     id=#3
     display_name=Overlay-AccoladeBG
     possessed_class=Image
     tracks [{
       type=MovieSceneFloatTrack
       display_name=Render Opacity
       sections [{type=MovieSceneFloatSection, channels [None {type=MovieSceneScriptingFloatChannel, keys [[25.4643,0],[32.1337,1]]}]}]
      }]
    }
    Text_EventDescription {
     id=#4
     possessed_class=CommonTextBlock
     tracks [
      {
       type=MovieSceneFloatTrack
       display_name=Render Opacity
       sections [{type=MovieSceneFloatSection, channels [None {type=MovieSceneScriptingFloatChannel, keys [[40.4497,0],[43.8973,1]]}]}]
      }
      {
       type=MovieScene2DTransformTrack
       sections [{
         type=MovieScene2DTransformSection
         channels [
          Scale.X {type=MovieSceneScriptingFloatChannel, keys [[40.4497,8],[44.7037,0.8094],[45.7963,0.9959],[47.119,1]]}
          Scale.Y {type=MovieSceneScriptingFloatChannel, keys [[40.4497,8],[44.7037,0.8094],[45.7963,0.9959],[47.119,1]]}
         ]
        }]
      }
     ]
    }
    Image_Glow {
     id=#5
     possessed_class=Image
     tracks [
      {
       type=MovieScene2DTransformTrack
       sections [{
         type=MovieScene2DTransformSection
         channels [
          Scale.X {type=MovieSceneScriptingFloatChannel, keys [[25.854,1],[29.7603,1],[47.119,1.2]]}
          Scale.Y {type=MovieSceneScriptingFloatChannel, keys [[25.854,1],[29.7603,1],[47.119,1.1]]}
         ]
        }]
      }
      {
       type=MovieSceneFloatTrack
       display_name=Render Opacity
       sections [{type=MovieSceneFloatSection, channels [None {type=MovieSceneScriptingFloatChannel, keys [[25.854,1],[61.9627,0]]}]}]
      }
     ]
    }
    Overlay-AugmentIcon {
     id=#6
     possessed_class=Overlay
     tracks [{
       type=MovieScene2DTransformTrack
       sections [{
         type=MovieScene2DTransformSection
         channels [
          Scale.X {type=MovieSceneScriptingFloatChannel, keys [[0,2],[4,0.6],[13.6013,0.7]]}
          Scale.Y {type=MovieSceneScriptingFloatChannel, keys [[0,2],[4,0.6],[13.6013,0.7]]}
         ]
        }]
      }]
    }
    "Text_Title (4)" {
     id=#7
     possessed_class=VerticalBox
     tracks [{
       type=MovieSceneFloatTrack
       display_name=Render Opacity
       sections [{type=MovieSceneFloatSection, channels [None {type=MovieSceneScriptingFloatChannel, keys [[0,0],[6,1]]}]}]
      }]
    }
   ]}
 }
 Outro {
  binding_count=1
  length_seconds=0.5016
  object=&self:Outro
  sequence {bindings [AccoladeImageWidget {
     id=#8
     display_name=Overlay_TopContainer
     possessed_class=AccoladeImageWidget_C
     tracks [{
       type=MovieSceneFloatTrack
       display_name=Render Opacity
       sections [{type=MovieSceneFloatSection, channels [None {type=MovieSceneScriptingFloatChannel, keys [[0,1],[10.0313,0]]}]}]
      }]
    }]}
 }
]
graphs [
 EventGraph {nodes [
   n1 = LoadAndSetIcon!() -> (OutputDelegate, then, IconAsset:Soft Object Reference) {is=title K2Node_CustomEvent}
   > n2 = Async Load Asset(Asset=@n1.IconAsset)
   n3 = Cast To MaterialInstance!(execute=@n2.Completed, Object=@n2) -> (then, CastFailed, AsMaterial Instance:Material Instance Object Reference) {is=title K2Node_DynamicCast}
   n4 = Get Image_Icon() -> (Image_Icon:Image Object Reference)
   > SetBrushFromMaterial(self=@n4, Material=@n3)
   n5 = Cast To Texture2D(execute=@n3.CastFailed, Object=@n2)
   > SetBrushFromTexture!(execute, self:Image Object Reference=@n4, Texture:Texture 2D Object Reference=@n5, bMatchSize:bool=false) -> (then)
   n6 = WaitForClose() {is=title K2Node_CustomEvent}
   n7 = "DoClose!"() {is=title K2Node_CustomEvent}
   n8 = HasMoreAugmentEvents!(self:Fort Player Augment Special Acquired Notification Widget Object Reference) -> (ReturnValue:bool)
   n9 = Set Timer by Event(execute=@n6.then, Delegate=@n7, Time=Select(Option 0:double=2.5, Option 1:double=0.1, Index=@n8) -> (ReturnValue:double), bLooping=false)
   > Set WaitForCloseTimer(WaitForCloseTimer:Timer Handle Structure=@n9) -> (Output_Get:Timer Handle Structure)
   HideAndCloseWidget() {is=title K2Node_CustomEvent}
   > SetVisibility(InVisibility=Collapsed)
   > CloseWidget(self:Fort Player Augment Special Acquired Notification Widget Object Reference)
   Event Construct()
   > n10 = PlaylistDataReadyAsync!(execute:Exec) -> (
    then:Exec
    OnReady:Exec
    GameState:Fort Game State Athena Object Reference
    Playlist:Fort Playlist Object Reference
    PlaylistContextTags:"Gameplay Tag Container Structure (by ref)"
   ) {is=title K2Node_AsyncAction}
   n11 = Handle Offset Translation on Gameplay Tags!(
    Exec:Exec=@n10.OnReady
    PlaylistGameplayTags:Array of Gameplay Tag Structures=BreakGameplayTagContainer!(GameplayTagContainer:"Gameplay Tag Container Structure (by ref)"=@n10.PlaylistContextTags) -> (GameplayTags:Array of Gameplay Tag Structures)
    PlaylistOffsetMap:Map of Gameplay Tag Structures to Vector 2D Structures=Get PlaylistOffsetMap() -> (PlaylistOffsetMap:Map of Gameplay Tag Structures to Vector 2D Structures)
   ) -> (OnOffsetFound:Exec, OnCompleted:Exec, TranslationOffset:Vector 2D Structure, FoundTag:Gameplay Tag Structure) {is=title K2Node_MacroInstance}
   > SetRenderTranslation(self=Get Overlay_TopContainer() -> (Overlay_TopContainer:Overlay Object Reference), Translation=@n11)
   Event OnStompedByOtherWidget() {is=title K2Node_Event}
   > StopAllAnimations(self:User Widget Object Reference)
   > SetVisibility(InVisibility=Collapsed)
   > Clear and Invalidate Timer by Handle(Handle=Get WaitForCloseTimer() -> (WaitForCloseTimer:Timer Handle Structure))
   Event OnInitialized()
   > BindToAnimationFinished(Animation=Get Outro() -> (Outro:Widget Animation Object Reference), Delegate=Create Event())
   n12 = GetOwningPlayer()
   > BindToAnimationFinished(
    Animation=Select(
     Option 0:Widget Animation Object Reference=Get Intro()
     Option 1:Widget Animation Object Reference=Get Intro_Mobile() -> (Intro_Mobile:Widget Animation Object Reference)
     Index=IsUsingTouch(PlayerController=@n12)
    ) -> (ReturnValue:Widget Animation Object Reference)
    Delegate=Create Event()
   )
   n13 = Event OpenWidget!() -> (OutputDelegate, then, ItemDef:Fort Player Augment Item Definition Object Reference) {is=title K2Node_Event}
   > n14 = Set ActiveAugment(ActiveAugment:Fort Player Augment Item Definition Object Reference=@n13.ItemDef) -> (Output_Get:Fort Player Augment Item Definition Object Reference)
   > n15 = Branch(Condition=NOT Boolean(A="Equal (Object)"(A=@n14, B=/PlotTwist/Items/Augments/PAID_Athena_PlotTwist_Force_SprintJump)))
   > Delay(Duration=6)
   > Parent: OpenWidget!(
    execute
    ItemDef:Fort Player Augment Item Definition Object Reference=Get ActiveAugment() -> (ActiveAugment:Fort Player Augment Item Definition Object Reference)
   ) -> (then) {is=title K2Node_CallParentFunction}
   > UpdateVisuals!(execute, self, AugmentDefinition:Fort Player Augment Item Definition Object Reference=Get ActiveAugment()) -> (then)
   n16 = GetOwningPlayer()
   > PlayAnimation(InAnimation=Select(Option 0=Get Intro(), Option 1=Get Intro_Mobile(), Index=IsUsingTouch(PlayerController=@n16)), PlaybackSpeed=1.5)
   HideAndCloseWidget(execute=@n15.else) {is=title K2Node_CallFunction}
   n17 = Event PreConstruct()
   > Branch(Condition=@n17.IsDesignTime)
   > n18 = Keys!(
    execute
    TargetMap:Map of Fort Player Augment Item Definition Object References to STRUCT Plot Twist Notification Data Structures=Get AugmentTypeData() -> (AugmentTypeData:Map of Fort Player Augment Item Definition Object References to STRUCT Plot Twist Notification Data Structures)
   ) -> (then, Keys:Array of Fort Player Augment Item Definition Object References)
   > UpdateVisuals(AugmentDefinition="Get (a copy)"(Array:Array of Fort Player Augment Item Definition Object References=@n18, Dimension 1=1) -> (Output:Fort Player Augment Item Definition Object Reference))
   > SetMobileStyle() {is=title K2Node_CallFunction}
   Event OnEventAdded() {is=title K2Node_Event}
   > Branch(Condition=Is Timer Active by Handle!(Handle:Timer Handle Structure=Get WaitForCloseTimer()) -> (ReturnValue:bool))
   > Clear and Invalidate Timer by Handle(Handle=Get WaitForCloseTimer())
   > "DoClose!"() {is=title K2Node_CallFunction}
   PlayAnimation(execute=@n7.then, InAnimation=Get Outro(), PlaybackSpeed=Select(Option 0:double=1, Option 1:double=4, Index=@n8) -> (ReturnValue:double))
  ]}
 GetNotificationDataFromAugment {nodes [
   n1 = GetNotificationDataFromAugment!() -> (then, AugmentDefinition:"Fort Player Augment Item Definition Object Reference (by ref)") {is=title K2Node_FunctionEntry}
   n2 = Get AugmentTypeData()
   > n3 = Values!(execute, TargetMap:Map of Fort Player Augment Item Definition Object References to STRUCT Plot Twist Notification Data Structures=@n2) -> (then, Values:Array of STRUCT Plot Twist Notification Data Structures)
   n4 = Find(
    TargetMap:Map of Fort Player Augment Item Definition Object References to STRUCT Plot Twist Notification Data Structures=@n2
    Key:"Fort Player Augment Item Definition Object Reference (by ref)"=@n1
   ) -> (Value:STRUCT Plot Twist Notification Data Structure)
   n5 = Get DefaultTypeIndex() -> (DefaultTypeIndex:int=0)
   > Return Node!(
    execute
    ReturnData:STRUCT Plot Twist Notification Data Structure=Select(
     Option 0:STRUCT Plot Twist Notification Data Structure="Get (a ref)"(Array:Array of STRUCT Plot Twist Notification Data Structures=@n3, Dimension 1=@n5) -> (Output:"STRUCT Plot Twist Notification Data Structure (by ref)")
     Option 1:STRUCT Plot Twist Notification Data Structure=@n4
     Index=@n4.ReturnValue
    ) -> (ReturnValue:STRUCT Plot Twist Notification Data Structure)
   )
  ]}
 SetMobileStyle {nodes [
   SetMobileStyle() {is=title K2Node_FunctionEntry}
   n1 = GetOwningPlayer()
   > Branch(Condition=AND Boolean(A=IsMobileGame(), B=IsUsingTouch(PlayerController=@n1)))
   n2 = Get Overlay-AugmentIcon() -> (Overlay-AugmentIcon:Overlay Object Reference)
   > SetRenderScale!(execute, self:Widget Object Reference=@n2, Scale:Vector 2D Structure=(X=0.7,Y=0.7)) -> (then)
   > SetPadding!(
    execute
    self:Vertical Box Slot Object Reference=SlotAsVerticalBoxSlot(Widget=@n2)
    InPadding_Left:float=0
    InPadding_Top:float=0
    InPadding_Right:float=0
    InPadding_Bottom:float=0
   ) -> (then)
   > SetPadding(
    self:Overlay Slot Object Reference=SlotAsOverlaySlot(Widget=Get Overlay_TopContainer())
    InPadding_Left=0
    InPadding_Top=75
    InPadding_Right=0
    InPadding_Bottom=0
   )
  ]}
 UpdateVisuals {nodes [
   n1 = UpdateVisuals!() -> (then, AugmentDefinition:Fort Player Augment Item Definition Object Reference) {is=title K2Node_FunctionEntry}
   > SetVisibility(InVisibility=HitTestInvisible)
   > n2 = Set NotificationData(NotificationData:STRUCT Plot Twist Notification Data Structure=GetNotificationDataFromAugment!(self, AugmentDefinition:"Fort Player Augment Item Definition Object Reference (by ref)"=@n1) -> (ReturnData:STRUCT Plot Twist Notification Data Structure) {is=title K2Node_CallFunction}) -> (Output_Get:STRUCT Plot Twist Notification Data Structure)
   n3 = Break STRUCT Plot Twist Notification Data!(STRUCT_PlotTwist_NotificationData:"STRUCT Plot Twist Notification Data Structure (by ref)"=@n2) -> (
    SoftIconMaterial_4_#9:Material Instance Soft Object Reference
    Title_7_#10:Text
    BannerGlowColor_18_#11:Linear Color Structure
    TextGlowColor_17_#12:Linear Color Structure
    TriangleColor_12_#13:Linear Color Structure
    BackgroundGlowColor_15_#14:Linear Color Structure
    IconGlow_22_#15:Linear Color Structure
    IconGlowOuterColor_25_#16:Linear Color Structure
   ) {is=title K2Node_BreakStruct}
   n4 = Get AugmentDefinition() -> (AugmentDefinition:Fort Player Augment Item Definition Object Reference)
   > LoadAndSetIcon!(
    execute
    self
    IconAsset:Soft Object Reference=Select(
     Option 0:Soft Object Reference=GetLargePreviewImage!(self:Fort Item Definition Object Reference=@n4) -> (ReturnValue:Texture 2D Soft Object Reference)
     Option 1:Soft Object Reference=@n3
     Index=IsValidSoftObjectReference(SoftObjectReference=@n3)
    ) -> (ReturnValue:Soft Object Reference)
   ) -> (then) {is=title K2Node_CallFunction}
   n5 = Get!() -> (
    NotificationData_SoftIconMaterial_4_#9:Material Instance Soft Object Reference
    NotificationData_Title_7_#10:Text
    NotificationData_BannerGlowColor_18_#11:Linear Color Structure
    NotificationData_TextGlowColor_17_#12:Linear Color Structure
    NotificationData_TriangleColor_12_#13:Linear Color Structure
    NotificationData_BackgroundGlowColor_15_#14:Linear Color Structure
    NotificationData_IconGlow_22_#15:Linear Color Structure
    NotificationData_IconGlowOuterColor_25_#16:Linear Color Structure
   )
   > SetBrushTintColor!(
    execute
    self:Image Object Reference=Get Image_Glow() -> (Image_Glow:Image Object Reference)
    TintColor_SpecifiedColor:Linear Color Structure=@n5.NotificationData_BannerGlowColor_18_#11
    TintColor_ColorUseRule:ESlateColorStylingMode Enum
   ) -> (then)
   > Set Outline Color!(
    In:Exec
    CommonTextWidget:Common Text Object Reference=Get Text_EventDescription() -> (Text_EventDescription:Common Text Object Reference)
    Color:Linear Color Structure=@n5.NotificationData_BannerGlowColor_18_#11
   ) -> (Out:Exec) {is=title K2Node_MacroInstance}
   > Set Outline Color(
    CommonTextWidget=Get Text_Title() -> (Text_Title:Common Text Object Reference)
    Color=@n5.NotificationData_BannerGlowColor_18_#11
   )
   > SetBrushTintColor(
    self=[Get ImageTriangleR() -> (ImageTriangleR:Image Object Reference), Get ImageTriangleL() -> (ImageTriangleL:Image Object Reference)]
    TintColor_SpecifiedColor=@n5.NotificationData_TriangleColor_12_#13
   )
   n6 = BreakColor(InColor=@n5.NotificationData_BackgroundGlowColor_15_#14)
   > SetColorAndOpacity!(
    execute
    self:Image Object Reference=[Get Image_TitleBG() -> (Image_TitleBG:Image Object Reference), Get Image_IconStars() -> (Image_IconStars:Image Object Reference)]
    InColorAndOpacity_R:float=@n6
    InColorAndOpacity_G:float=@n6.G
    InColorAndOpacity_B:float=@n6.B
    InColorAndOpacity_A:float=@n6.A
   ) -> (then)
   n7 = Get()
   > "SetText (Text)"(self=Get Text_Title(), InText=@n7.NotificationData_Title_7_#10)
   > n8 = GetDynamicMaterial(self=Get Image_IconGlow() -> (Image_IconGlow:Image Object Reference))
   > SetVectorParameterValue(self=@n8, ParameterName=Inner Color, Value=@n7.NotificationData_IconGlow_22_#15)
   > SetVectorParameterValue(self=@n8, ParameterName=Outer Color, Value=@n7.NotificationData_IconGlowOuterColor_25_#16)
   > SetMobileStyle() {is=title K2Node_CallFunction}
  ]}
]
variables [
 ActiveAugment {type=FortPlayerAugmentItemDefinition, default=null}
 AugmentTypeData {default {
   /Script/FortniteGame.FortPlayerAugmentItemDefinition'/PlotTwist/Items/Augments/PAID_Athena_AmberVine_Force.PAID_Athena_AmberVine_Force' {
    softIconMaterial=/PlotTwist/Items/RiftGate/Assets/NPCs/RiftGate/UI/Materials/MI_UI_Icon_FP_AmberVine
    title="Sith Training Complete!"
    backgroundGlowColor=<rgba 0.8244 0.0415 0.004 1>
    bannerGlowColor=<rgba 1 0 0 1>
    iconGlow=<rgba 0.9077 0 0 1>
    iconGlowOuterColor=<rgba 0.0982 0.0029 0.0046 1>
    textGlowColor=<rgba 1 0 0 1>
    triangleColor=<rgba 1 0.5 0.5 1>
   }
   &FortPlayerAugmentItemDefinition'/PlotTwist/Items/Augments/PAID_Athena_PlotTwist_Force_Projectile.PAID_Athena_PlotTwist_Force_Projectile' {
    softIconMaterial=/PlotTwist/Items/RiftGate/Assets/NPCs/RiftGate/UI/Materials/MI_UI_Icon_FP_Throw
    title="Sith Training Complete!"
    backgroundGlowColor=<rgba 0.8244 0.0415 0.004 1>
    bannerGlowColor=<rgba 1 0 0 1>
    iconGlow=<rgba 0.9077 0 0 1>
    iconGlowOuterColor=<rgba 0.0982 0.0029 0.0046 1>
    textGlowColor=<rgba 1 0 0 1>
    triangleColor=<rgba 1 0.5 0.5 1>
   }
   &FortPlayerAugmentItemDefinition'/PlotTwist/Items/Augments/PAID_Athena_PlotTwist_Force_Push.PAID_Athena_PlotTwist_Force_Push' {
    softIconMaterial=/PlotTwist/Items/RiftGate/Assets/NPCs/RiftGate/UI/Materials/MI_UI_Icon_FP_Push
    title="Jedi Training Complete!"
    backgroundGlowColor=<rgba 0 1 0 1>
    bannerGlowColor=<rgba 0 1 0 1>
    iconGlow=<rgba 0 0.872 0 1>
    iconGlowOuterColor=<rgba 0 0.0685 0 1>
    textGlowColor=<rgba 0 1 0 1>
    triangleColor=<rgba 0.5029 1 0.5029 1>
   }
   &FortPlayerAugmentItemDefinition'/PlotTwist/Items/Augments/PAID_Athena_PlotTwist_Force_Rip.PAID_Athena_PlotTwist_Force_Rip' {
    softIconMaterial=/PlotTwist/Items/RiftGate/Assets/NPCs/RiftGate/UI/Materials/MI_UI_Icon_FP_Pull
    title="Jedi Training Complete!"
    backgroundGlowColor=<rgba 0 0.1746 0.5841 1>
    bannerGlowColor=<rgba 0 0.299 1 1>
    iconGlow=<rgba 0 0.4727 1 1>
    iconGlowOuterColor=<rgba 0 0.0319 0.2122 1>
    textGlowColor=<rgba 0 0.1746 0.5841 1>
    triangleColor=<rgba 0.5029 0.5029 1 1>
   }
   &FortPlayerAugmentItemDefinition'/PlotTwist/Items/Augments/PAID_Athena_ThunderBarn_Force_Pull.PAID_Athena_ThunderBarn_Force_Pull' {
    softIconMaterial=&MI_UI_Icon_FP_Pull
    title="Jedi Training Complete!"
    backgroundGlowColor=<rgba 0.3 0 0.9 1>
    bannerGlowColor=<rgba 0.6654 0 1 1>
    iconGlow=<rgba 0.6654 0 1 1>
    iconGlowOuterColor=<rgba 0.0333 0 0.1 1>
    textGlowColor=<rgba 0.6654 0 1 1>
    triangleColor=<rgba 0.8327 0.5 1 1>
   }
   &FortPlayerAugmentItemDefinition'/PlotTwist/Items/Augments/PAID_Athena_ThunderBarn_Force_Push.PAID_Athena_ThunderBarn_Force_Push' {
    softIconMaterial=&MI_UI_Icon_FP_Push
    title="Jedi Training Complete!"
    backgroundGlowColor=<rgba 0 0.1746 0.5841 1>
    bannerGlowColor=<rgba 0 0.299 1 1>
    iconGlow=<rgba 0 0.4727 1 1>
    iconGlowOuterColor=<rgba 0 0.0319 0.2122 1>
    textGlowColor=<rgba 0 0.1746 0.5841 1>
    triangleColor=<rgba 0.5029 0.5029 1 1>
   }
   &FortPlayerAugmentItemDefinition'/PlotTwist/Items/HeavyLamp/PAID_Athena_HeavyLamp_Force.PAID_Athena_HeavyLamp_Force' {
    softIconMaterial=/PlotTwist/Items/RiftGate/Assets/NPCs/RiftGate/UI/Materials/MI_UI_Icon_FP_HeavyLamp
    title="Sith Training Complete!"
    backgroundGlowColor=<rgba 0.8244 0.0415 0.004 1>
    bannerGlowColor=<rgba 1 0 0 1>
    iconGlow=<rgba 0.9077 0 0 1>
    iconGlowOuterColor=<rgba 0.0982 0.0029 0.0046 1>
    textGlowColor=<rgba 1 0 0 1>
    triangleColor=<rgba 1 0.5 0.5 1>
   }
  }}
 DefaultTypeIndex {type=integer, default=1}
 PlaylistOffsetMap {default {
   "(TagName=\"Athena.Playlist.Argon\")"=<0 120>
   "(TagName=\"Athena.Playlist.CobaltRespawn\")"=<0 70>
   "(TagName=\"Athena.Playlist.Infiltration\")"=<0 120>
   "(TagName=\"Athena.Playlist.Payback\")"=<0 180>
   "(TagName=\"Athena.Playlist.Phosphorus\")"=<0 40>
   "(TagName=\"Athena.Playlist.UraniumP\")"=<0 120>
  }}
 WaitForCloseTimer {type=TimerHandle}
]
widgets {
 info {
  inheritedWidgetCount=1
  parentClass=/Script/PlayerAugmentsUI.FortPlayerAugmentSpecialAcquiredNotificationWidget
  rootWidgetClass=&Overlay
  widgetCount=20
 }
 widgets [
  Overlay_TopOffset {widgetClassPath=&Overlay, bIsVariable=true, widget=Overlay_TopOffset}
  Overlay_TopContainer {widgetClassPath=&Overlay, bIsVariable=true, parent=Overlay_TopOffset, widget=Overlay_TopContainer, slot {class=OverlaySlot}}
  Image_shadow {
   widgetClassPath=&Image
   parent=Overlay_TopContainer
   widget=Image_shadow
   properties {
    brush {
     bIsDynamicallyLoaded=false
     drawAs=Image
     imageType=NoImage
     mirroring=NoMirror
     resourceName=None
     resourceObject=/Game/Athena/HUD/LTM/Wax/Materials/M_UI_TextGlow_Inst
     tiling=NoTile
     imageSize=<279.4116 128.8524>
     margin=<ltrb 0 0 0 0>
     outlineSettings {
      bUseBrushTransparency=false
      roundingType=HalfHeightRadius
      width=0
      color {colorUseRule=UseColor_Specified, specifiedColor=<rgba 0 0 0 0>}
      cornerRadii=<0 0 0 0>
     }
     tintColor {colorUseRule=UseColor_Specified, specifiedColor=<rgba 1 1 1 1>}
     uVRegion {bIsValid=false, max=<0 0>, min=<0 0>}
    }
    colorAndOpacity=<rgba 0.0156 0.0073 0.0032 0.2>
    renderTransform {angle=0, scale=<1 1>, shear=<0 0>, translation=<0 -5>}
   }
   slot {class=OverlaySlot, properties {horizontalAlignment=HAlign_Fill, verticalAlignment=VAlign_Fill, padding=<ltrb -125 -125 -125 -125>}}
  }
  VerticalBox_Container {widgetClassPath=&VerticalBox, parent=Overlay_TopContainer, widget=VerticalBox_Container, slot {class=OverlaySlot}}
  Overlay-AugmentIcon {
   widgetClassPath=&Overlay
   bIsVariable=true
   parent=VerticalBox_Container
   widget=Overlay-AugmentIcon
   slot {class=VerticalBoxSlot, properties {horizontalAlignment=HAlign_Center, padding=<ltrb 0 0 0 12>}}
  }
  Image_Icon {
   widgetClassPath=&Image
   bIsVariable=true
   parent=Overlay-AugmentIcon
   widget=Image_Icon
   properties {brush {
     bIsDynamicallyLoaded=false
     drawAs=Image
     imageType=NoImage
     mirroring=NoMirror
     resourceName=None
     resourceObject=null
     tiling=NoTile
     imageSize=<128 128>
     margin=<ltrb 0 0 0 0>
     outlineSettings {
      bUseBrushTransparency=false
      roundingType=HalfHeightRadius
      width=0
      color {colorUseRule=UseColor_Specified, specifiedColor=<rgba 0 0 0 0>}
      cornerRadii=<0 0 0 0>
     }
     tintColor {colorUseRule=UseColor_Specified, specifiedColor=<rgba 1 1 1 1>}
     uVRegion {bIsValid=false, max=<0 0>, min=<0 0>}
    }}
   slot {class=OverlaySlot, properties {horizontalAlignment=HAlign_Center, verticalAlignment=VAlign_Center}}
  }
  Image_IconStars {
   widgetClassPath=&Image
   bIsVariable=true
   parent=Overlay-AugmentIcon
   widget=Image_IconStars
   properties {
    brush {
     bIsDynamicallyLoaded=false
     drawAs=Image
     imageType=NoImage
     mirroring=NoMirror
     resourceName=None
     resourceObject=/PlotTwist/Items/RiftGate/Assets/NPCs/RiftGate/UI/Materials/M_UI_ForcePower_Icon_Bg
     tiling=NoTile
     imageSize=<200 200>
     margin=<ltrb 0 0 0 0>
     outlineSettings {
      bUseBrushTransparency=false
      roundingType=HalfHeightRadius
      width=0
      color {colorUseRule=UseColor_Specified, specifiedColor=<rgba 0 0 0 0>}
      cornerRadii=<0 0 0 0>
     }
     tintColor {colorUseRule=UseColor_Specified, specifiedColor=<rgba 1 1 1 1>}
     uVRegion {bIsValid=false, max=<0 0>, min=<0 0>}
    }
    colorAndOpacity=<rgba 0 0 1 1>
   }
   slot {class=OverlaySlot, properties {horizontalAlignment=HAlign_Center, verticalAlignment=VAlign_Center, padding=<ltrb -100 -100 -100 -100>}}
  }
  Image_IconGlow {
   widgetClassPath=&Image
   bIsVariable=true
   parent=Overlay-AugmentIcon
   widget=Image_IconGlow
   properties {
    brush {
     bIsDynamicallyLoaded=false
     drawAs=Image
     imageType=NoImage
     mirroring=NoMirror
     resourceName=None
     resourceObject=/PlotTwist/Items/RiftGate/Assets/NPCs/RiftGate/UI/Materials/MI_UI_ForcePower_Glow
     tiling=NoTile
     imageSize=<200 200>
     margin=<ltrb 0 0 0 0>
     outlineSettings {
      bUseBrushTransparency=false
      roundingType=HalfHeightRadius
      width=0
      color {colorUseRule=UseColor_Specified, specifiedColor=<rgba 0 0 0 0>}
      cornerRadii=<0 0 0 0>
     }
     tintColor {colorUseRule=UseColor_Specified, specifiedColor=<rgba 1 1 1 1>}
     uVRegion {bIsValid=false, max=<0 0>, min=<0 0>}
    }
    colorAndOpacity=<rgba 1 1 1 0.5>
    renderTransform {angle=0, scale=<0.7 0.65>, shear=<0 0>, translation=<0 0>}
   }
   slot {class=OverlaySlot, properties {horizontalAlignment=HAlign_Center, verticalAlignment=VAlign_Center, padding=<ltrb -100 -100 -100 -100>}}
  }
  Overlay_Title {
   widgetClassPath=&Overlay
   parent=VerticalBox_Container
   widget=Overlay_Title
   slot {class=VerticalBoxSlot, properties {horizontalAlignment=HAlign_Center, padding=<ltrb 0 -8 0 0>}}
  }
  Overlay-AugmentBG {
   widgetClassPath=&Overlay
   bIsVariable=true
   parent=Overlay_Title
   widget=Overlay-AugmentBG
   slot {class=OverlaySlot, properties {horizontalAlignment=HAlign_Fill, verticalAlignment=VAlign_Center, padding=<ltrb -125 0 -125 0>}}
  }
  SizeBoxTriangles {
   widgetClassPath=&SizeBox
   parent=Overlay-AugmentBG
   widget=SizeBoxTriangles
   properties {bOverride_HeightOverride=true, bOverride_WidthOverride=true, heightOverride=49.1371, widthOverride=121.8743}
   slot {class=OverlaySlot, properties {horizontalAlignment=HAlign_Center, verticalAlignment=VAlign_Center, padding=<ltrb 0 -50 0 -50>}}
  }
  HorizontalBox_0 {widgetClassPath=&HorizontalBox, parent=SizeBoxTriangles, widget=HorizontalBox_0, slot {class=SizeBoxSlot}}
  ImageTriangleL {
   widgetClassPath=&Image
   bIsVariable=true
   parent=HorizontalBox_0
   widget=ImageTriangleL
   properties {
    brush {
     bIsDynamicallyLoaded=false
     drawAs=Image
     imageType=NoImage
     mirroring=NoMirror
     resourceName=None
     resourceObject=/Game/UI/HUD/Materials/M_FX_TriangleSpiral
     tiling=NoTile
     imageSize=<125 85>
     margin=<ltrb 0 0 0 0>
     outlineSettings {
      bUseBrushTransparency=false
      roundingType=HalfHeightRadius
      width=0
      color {colorUseRule=UseColor_Specified, specifiedColor=<rgba 0 0 0 0>}
      cornerRadii=<0 0 0 0>
     }
     tintColor {colorUseRule=UseColor_Specified, specifiedColor=<rgba 0.9427 1 0 1>}
     uVRegion {bIsValid=false, max=<0 0>, min=<0 0>}
    }
    renderTransform {angle=0, scale=<-1 1>, shear=<0 0>, translation=<0 0>}
    renderTransformPivot=<0.5 1>
   }
   slot {class=HorizontalBoxSlot, properties {verticalAlignment=VAlign_Center, padding=<ltrb -200 0 0 0>, size {sizeRule=Fill, value=0.5}}}
  }
  ImageTriangleR {
   widgetClassPath=&Image
   bIsVariable=true
   parent=HorizontalBox_0
   widget=ImageTriangleR
   properties {
    brush {
     bIsDynamicallyLoaded=false
     drawAs=Image
     imageType=NoImage
     mirroring=NoMirror
     resourceName=None
     resourceObject=&M_FX_TriangleSpiral
     tiling=NoTile
     imageSize=<125 85>
     margin=<ltrb 0 0 0 0>
     outlineSettings {
      bUseBrushTransparency=false
      roundingType=HalfHeightRadius
      width=0
      color {colorUseRule=UseColor_Specified, specifiedColor=<rgba 0 0 0 0>}
      cornerRadii=<0 0 0 0>
     }
     tintColor {colorUseRule=UseColor_Specified, specifiedColor=<rgba 0.9427 1 0 1>}
     uVRegion {bIsValid=false, max=<0 0>, min=<0 0>}
    }
    renderTransformPivot=<0.5 1>
   }
   slot {class=HorizontalBoxSlot, properties {verticalAlignment=VAlign_Center, padding=<ltrb 0 0 -200 0>, size {sizeRule=Fill, value=0.5}}}
  }
  Image_TitleBG {
   widgetClassPath=&Image
   bIsVariable=true
   parent=Overlay-AugmentBG
   widget=Image_TitleBG
   properties {
    brush {
     bIsDynamicallyLoaded=false
     drawAs=Image
     imageType=NoImage
     mirroring=NoMirror
     resourceName=None
     resourceObject=/PlotTwist/Items/RiftGate/Assets/NPCs/RiftGate/UI/Materials/MI_UI_AnamorphicFlare_ForcePowers
     tiling=NoTile
     imageSize=<32 32>
     margin=<ltrb 0 0 0 0>
     outlineSettings {
      bUseBrushTransparency=false
      roundingType=HalfHeightRadius
      width=0
      color {colorUseRule=UseColor_Specified, specifiedColor=<rgba 0 0 0 0>}
      cornerRadii=<0 0 0 0>
     }
     tintColor {colorUseRule=UseColor_Specified, specifiedColor=<rgba 1 1 1 1>}
     uVRegion {bIsValid=false, max=<0 0>, min=<0 0>}
    }
    colorAndOpacity=<rgba 1 0.3529 0 0.39>
    renderTransform {angle=0, scale=<1 1>, shear=<0 0>, translation=<2 -2>}
   }
   slot {class=OverlaySlot, properties {horizontalAlignment=HAlign_Fill}}
  }
  Image_Glow {
   widgetClassPath=&Image
   bIsVariable=true
   parent=Overlay-AugmentBG
   widget=Image_Glow
   properties {brush {
     bIsDynamicallyLoaded=false
     drawAs=Image
     imageType=NoImage
     mirroring=NoMirror
     resourceName=None
     resourceObject=/Game/Athena/HUD/LTM/Wax/Materials/M_UI_TextGlow_Add
     tiling=NoTile
     imageSize=<32 32>
     margin=<ltrb 0 0 0 0>
     outlineSettings {
      bUseBrushTransparency=false
      roundingType=HalfHeightRadius
      width=0
      color {colorUseRule=UseColor_Specified, specifiedColor=<rgba 0 0 0 0>}
      cornerRadii=<0 0 0 0>
     }
     tintColor {colorUseRule=UseColor_Specified, specifiedColor=<rgba 0 0.1746 0.5841 1>}
     uVRegion {bIsValid=false, max=<0 0>, min=<0 0>}
    }}
   slot {class=OverlaySlot, properties {horizontalAlignment=HAlign_Fill, verticalAlignment=VAlign_Fill, padding=<ltrb 0 -20 0 -20>}}
  }
  Text_Title {
   widgetClassPath=&CommonTextBlock
   bIsVariable=true
   parent=Overlay_Title
   widget=Text_Title
   properties {
    mobileFontSizeMultiplier=1.2
    style=/Game/Athena/HUD/TextStyles/TextStyle-HUD-XPDescriptor.TextStyle-HUD-XPDescriptor_C
    text="Jedi Training Complete!"
    textTransformPolicy=ToUpper
    font {
     bForceMonospaced=false
     bMaterialIsStencil=false
     fontMaterial=null
     fontObject=&Burbank
     letterSpacing=0
     monospacedWidth=1
     size=20
     skewAmount=0
     typefaceFontName=Burbank-PlayerPower
     outlineSettings {
      bApplyOutlineToDropShadows=false
      bMiteredCorners=false
      bSeparateFillAlpha=true
      outlineMaterial=&M-FontShadow
      outlineSize=0
      outlineColor=<rgba 0.1094 0.0129 0 0.5>
     }
    }
    renderTransform {angle=0, scale=<1 1>, shear=<-8 0>, translation=<0 0>}
    renderTransformPivot=<0.5 -0.5>
    shadowColorAndOpacity=<rgba 0.0052 0.0052 0.0056 0.4>
    shadowOffset=<2 2>
   }
   slot {class=OverlaySlot, properties {horizontalAlignment=HAlign_Center}}
  }
  ScaleBox_1 {
   widgetClassPath=&ScaleBox
   parent=VerticalBox_Container
   widget=ScaleBox_1
   properties {stretch=UserSpecified, userSpecifiedScale=0.9}
   slot {class=VerticalBoxSlot, properties {horizontalAlignment=HAlign_Center, padding=<ltrb 0 -8 0 0>}}
  }
  Text_EventDescription {
   widgetClassPath=&CommonTextBlock
   bIsVariable=true
   parent=ScaleBox_1
   widget=Text_EventDescription
   properties {
    mobileFontSizeMultiplier=1.05
    style=&TextStyle-HUD-XPDescriptor_C
    text="Force Abilities Available!"
    font {
     bForceMonospaced=false
     bMaterialIsStencil=false
     fontMaterial=null
     fontObject=&Burbank
     letterSpacing=0
     monospacedWidth=1
     size=20
     skewAmount=0
     typefaceFontName=Burbank-PlayerPower
     outlineSettings {
      bApplyOutlineToDropShadows=false
      bMiteredCorners=false
      bSeparateFillAlpha=true
      outlineMaterial=&M-FontShadow
      outlineSize=0
      outlineColor=<rgba 0.1094 0.0129 0 0.5>
     }
    }
    renderTransform {angle=0, scale=<1 1>, shear=<-8 0>, translation=<0 0>}
    renderTransformPivot=<0.5 -0.5>
    shadowColorAndOpacity=<rgba 0.0052 0.0052 0.0056 0.4>
    shadowOffset=<2 2>
   }
   slot {class=ScaleBoxSlot, properties {verticalAlignment=VAlign_Top}}
  }
  TutorialHighlight {widgetClassPath=&TutorialHighlight}
 ]
}
